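# Widget Blueprint for a player portrait in the in-game Leto UI, built on the native LetoPlayerAvatar class.
# A dark navy border wraps a ScaleBox and Overlay that stack a radial gradient, the inherited social avatar icon, a translucent blue tint and a framed edge material.
# The event graph holds only unconnected PreConstruct, Construct and Tick events, so the Blueprint adds layout and styling but no logic.
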WidgetBlueprint /FortniteCoreUI/Athena/UI/InGame/Leto/LetoPlayerAvatar
graphs [EventGraph {nodes [
   Event PreConstruct()
   Event Construct()
   Event Tick()
  ]}]
variables []
widgets {
 info {inheritedWidgetCount=1, parentClass=/Script/FortniteUI.LetoPlayerAvatar, rootWidgetClass=&CommonBorder, widgetCount=7}
 widgets [
  CommonBorder_0 {
   widgetClassPath=&CommonBorder
   widget=CommonBorder_0
   properties {
    background {
     bIsDynamicallyLoaded=false
     drawAs=Image
     imageType=NoImage
     mirroring=NoMirror
     resourceName=None
     resourceObject=null
     tiling=NoTile
     imageSize=<32 32>
     margin=<ltrb 0 0 0 0>
     outlineSettings {
      bUseBrushTransparency=false
      roundingType=HalfHeightRadius
      width=0
      color {colorUseRule=UseColor_Specified, specifiedColor=<rgba 0 0 0 0>}
      cornerRadii=<0 0 0 0>
     }
     tintColor {colorUseRule=UseColor_Specified, specifiedColor=<rgba 1 1 1 1>}
     uVRegion {bIsValid=false, max=<0 0>, min=<0 0>}
    }
    brushColor=<rgba 0 0.0116 0.0612 1>
    padding=<ltrb 0 30 0 30>
   }
  }
  ScaleBox_0 {
   widgetClassPath=&ScaleBox
   parent=CommonBorder_0
   widget=ScaleBox_0
   slot {class=BorderSlot, properties {padding=<ltrb 0 30 0 30>}}
  }
  Overlay_0 {
   widgetClassPath=&Overlay
   parent=ScaleBox_0
   widget=Overlay_0
   properties {visibility=HitTestInvisible}
   slot {class=ScaleBoxSlot}
  }
  Gradient_Bkg {
   widgetClassPath=&Image
   parent=Overlay_0
   widget=Gradient_Bkg
   properties {brush {
     bIsDynamicallyLoaded=false
     drawAs=Image
     imageType=NoImage
     mirroring=NoMirror
     resourceName=None
     resourceObject=/Game/Athena/UI/InGame/Leto/M_UI_Leto_RadialGradient_Gapfill1
     tiling=NoTile
     imageSize=<32 32>
     margin=<ltrb 0 0 0 0>
     outlineSettings {
      bUseBrushTransparency=false
      roundingType=HalfHeightRadius
      width=0
      color {colorUseRule=UseColor_Specified, specifiedColor=<rgba 0 0 0 0>}
      cornerRadii=<0 0 0 0>
     }
     tintColor {colorUseRule=UseColor_Specified, specifiedColor=<rgba 1 1 1 1>}
     uVRegion {bIsValid=false, max=<0 0>, min=<0 0>}
    }}
   slot {class=OverlaySlot, properties {horizontalAlignment=HAlign_Fill, verticalAlignment=VAlign_Fill}}
  }
  AvatarIcon {
   widgetClassPath=/Script/FortniteUI.FortSocialAvatarIcon
   bInherited=true
   parent=Overlay_0
   widget=AvatarIcon
   properties {
    bIsImageForDesignerPreviewOnly=true
    designerPreviewItem=/BRCosmetics/Athena/Items/Cosmetics/Characters/CID_298_Athena_Commando_F_IceMaiden
    brush {
     bIsDynamicallyLoaded=false
     drawAs=Image
     imageType=NoImage
     mirroring=NoMirror
     resourceName=None
     resourceObject=/BRCosmetics/UI/Foundation/Textures/Icons/Heroes/Athena/Soldier/T-Soldier-HID-298-Athena-Commando-F-IceMaiden-L
     tiling=NoTile
     imageSize=<420 420>
     margin=<ltrb 0 0 0 0>
     outlineSettings {
      bUseBrushTransparency=false
      roundingType=HalfHeightRadius
      width=0
      color {colorUseRule=UseColor_Specified, specifiedColor=<rgba 0 0 0 0>}
      cornerRadii=<0 0 0 0>
     }
     tintColor {colorUseRule=UseColor_Specified, specifiedColor=<rgba 1 1 1 1>}
     uVRegion {bIsValid=false, max=<0 0>, min=<0 0>}
    }
   }
   slot {class=OverlaySlot, properties {horizontalAlignment=HAlign_Center, verticalAlignment=VAlign_Center, padding=<ltrb 30 30 30 30>}}
  }
  BlueKnockbackTint {
   widgetClassPath=&Image
   parent=Overlay_0
   widget=BlueKnockbackTint
   properties {visibility=SelfHitTestInvisible, colorAndOpacity=<rgba 0 0.2122 0.6308 0.25>}
   slot {class=OverlaySlot, properties {horizontalAlignment=HAlign_Fill, verticalAlignment=VAlign_Fill}}
  }
  Angled_Border {
   widgetClassPath=&Image
   parent=Overlay_0
   widget=Angled_Border
   properties {
    visibility=SelfHitTestInvisible
    brush {
     bIsDynamicallyLoaded=false
     drawAs=Box
     imageType=NoImage
     mirroring=NoMirror
     resourceName=None
     resourceObject=/Game/Athena/UI/InGame/Leto/M_UI_Leto_UserPortrait_Frame
     tiling=NoTile
     imageSize=<32 32>
     margin=<ltrb 0 0 0 0>
     outlineSettings {
      bUseBrushTransparency=false
      roundingType=HalfHeightRadius
      width=0
      color {colorUseRule=UseColor_Specified, specifiedColor=<rgba 0 0 0 0>}
      cornerRadii=<0 0 0 0>
     }
     tintColor {colorUseRule=UseColor_Specified, specifiedColor=<rgba 1 1 1 1>}
     uVRegion {bIsValid=false, max=<0 0>, min=<0 0>}
    }
   }
   slot {class=OverlaySlot, properties {horizontalAlignment=HAlign_Fill, verticalAlignment=VAlign_Fill}}
  }
 ]
}
