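# LetoHudOverlay is a two-player in-game HUD overlay widget that places a LetoPlayerAvatar for each player.
# A vertical box splits the space into two equal fill rows, player one above and player two below.
# Fill spacers mirror the avatars: player one's sits at the right, player two's at the left.
# All widgets are inherited from the native LetoHudOverlay parent, and the event graph is empty.
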WidgetBlueprint /FortniteCoreUI/Athena/UI/InGame/Leto/LetoHudOverlay
graphs [EventGraph {links [], nodes []}]
variables []
widgets {
 info {inheritedWidgetCount=7, parentClass=/Script/FortniteUI.LetoHudOverlay, rootWidgetClass=&VerticalBox, widgetCount=7}
 widgets [
  VBox_Container {widgetClassPath=&VerticalBox, bInherited=true, widget=VBox_Container}
  HBox_PlayerOne {
   widgetClassPath=&HorizontalBox
   bInherited=true
   parent=VBox_Container
   widget=HBox_PlayerOne
   slot {class=VerticalBoxSlot, properties {size {sizeRule=Fill, value=0.5}}}
  }
  Spacer_PlayerOne {
   widgetClassPath=&Spacer
   bInherited=true
   parent=HBox_PlayerOne
   widget=Spacer_PlayerOne
   slot {class=HorizontalBoxSlot, properties {size {sizeRule=Fill, value=0.75}}}
  }
  LetoAvatar_PlayerOne {
   widgetClassPath=&/LetoPlayerAvatar
   bInherited=true
   parent=HBox_PlayerOne
   widget=LetoAvatar_PlayerOne
   slot {class=HorizontalBoxSlot}
  }
  HBox_PlayerTwo {
   widgetClassPath=&HorizontalBox
   bInherited=true
   parent=VBox_Container
   widget=HBox_PlayerTwo
   slot {class=VerticalBoxSlot, properties {size {sizeRule=Fill, value=0.5}}}
  }
  LetoAvatar_PlayerTwo {
   widgetClassPath=&LetoPlayerAvatar
   bInherited=true
   parent=HBox_PlayerTwo
   widget=LetoAvatar_PlayerTwo
   slot {class=HorizontalBoxSlot}
  }
  Spacer_PlayerTwo {
   widgetClassPath=&Spacer
   bInherited=true
   parent=HBox_PlayerTwo
   widget=Spacer_PlayerTwo
   slot {class=HorizontalBoxSlot, properties {size {sizeRule=Fill, value=0.75}}}
  }
 ]
}
